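# Simple material named Cornea for the eye surface of the M_Med_Soldier_04 male player body.
# It is Default Lit and flagged for skeletal mesh and mesh deformer use.
# The graph only scales a mid-grey vector parameter's RGB by a small constant to darken it, with no material output link shown.
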Material /Game/Characters/Player/Male/Medium/Bodies/M_Med_Soldier_04/Materials/Cornea
graph {
 properties {bUsedWithMeshDeformer=true, bUsedWithSkeletalMesh=true, pixelDepthOffsetMode=PDOM_Legacy, usedShadingModels=Default Lit}
 nodes [Multiply(A=VectorParameter() {defaultValue=<rgba 0.5 0.5 0.5 0>}.RGB, B=Constant() {r=0.2})]
}
