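# Default Lit eye material for the Soldier 04 medium male player body, flagged for skeletal mesh and mesh deformer use.
# The soldier eye texture is sampled, and a scalar Emissive parameter scales its RGB, by name a tunable low-level glow.
# A second sample of the same texture and two scalar constants complete the small graph, whose material output connections are not recorded.
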Material /Game/Characters/Player/Male/Medium/Bodies/M_Med_Soldier_04/Materials/Eye
graph {
 properties {bUsedWithMeshDeformer=true, bUsedWithSkeletalMesh=true, pixelDepthOffsetMode=PDOM_Legacy, usedShadingModels=Default Lit}
 nodes [
  Constant() {r=0.1333}
  Constant() {r=0.3}
  Multiply(
   A=TextureSample() {texture=/Game/Characters/Player/Male/Medium/Bodies/M_Med_Soldier_04/Textures/M_Med_Soldier_Eye_01}.RGB
   B=ScalarParameter() {defaultValue=0.15, parameterName=Emissive}
  )
  TextureSample() {texture=&M_Med_Soldier_Eye_01}
 ]
}
