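# Animation Blueprint for the holster mesh part of the M_Med_Soldier_04 male player character.
# On initialization it casts the owning actor to FortPlayerPawn and stores that pawn's mesh as the source to copy.
# The AnimGraph copies the pose from that mesh as the base.
# A per-bone layered blend filtered at holster_mstr overlays DefaultSlot montages, fed by a local-space reference pose, on that branch.
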AnimBlueprint /Game/Characters/Player/Male/Medium/Bodies/M_Med_Soldier_04/Meshes/Parts/SK_M_Med_Soldier_04_Holster_Skeleton_AnimBlueprint
target_skeleton=&/SK_M_Med_Soldier_04_Holster_Skeleton
graphs [
 AnimGraph {nodes [
   n1 = Get MeshToCopy() -> (MeshToCopy:Skeletal Mesh Component Object Reference)
   n2 = Local Space Ref Pose()
   Output Pose(Result=Layered blend per bone(BasePose=Copy Pose From Mesh(SourceMeshComponent=@n1), BlendPoses_0=Slot 'DefaultSlot'(Source=@n2)) {layerSetup [{branchFilters [{boneName=holster_mstr}]}]}) {layerGroup=None}
  ]}
 EventGraph {nodes [
   Event BlueprintInitializeAnimation()
   n1 = GetOwningActor()
   > Set MeshToCopy(MeshToCopy:Skeletal Mesh Component Object Reference=Get Mesh(self=Cast To FortPlayerPawn!(Object=@n1) -> (AsFort Player Pawn, bSuccess:bool=~))) -> (Output_Get:Skeletal Mesh Component Object Reference)
   TryGetPawnOwner()
   Event BlueprintUpdateAnimation()
  ]}
]
variables [MeshToCopy {type=SkeletalMeshComponent}]
