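# Physics asset for the dynamic scarf and tag bones of the Medium male Soldier 04 player body, named as a collision-override test.
# Box and small capsule bodies form the accessories, anchored to one large capsule on spine_03.
# The front scarf master swings and twists freely, leading to a limited chain that branches into the scarf tip and tags, which split left and right.
# The back scarf attaches separately to the spine with tighter limits.
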PhysicsAsset /Game/Characters/Player/Male/Medium/Bodies/M_Med_Soldier_04/Meshes/SK_M_Med_Soldier_04_Dyn_TestOverrideCollision
bodies {
 dyn_scarf_02 [{shapeType=Box, extentX=5.8889, extentY=5.9961, extentZ=0.6545, center=<3.4138 0 -0.0001>}]
 dyn_scarf_b [{shapeType=Capsule, length=2.0964, radius=1.6852, center=<0 0 -0.3019>, rotation=<rot 90 0 0>}]
 dyn_scarf_f_01 [{shapeType=Box, extentX=7.8593, extentY=7.5, extentZ=1.3703, center=<7.396 0.0002 -0.0001>}]
 dyn_scarf_f_MSTR [{shapeType=Capsule, length=0.505, radius=0.505, rotation=<rot 0 -90 0>}]
 dyn_tag_l [{shapeType=Box, extentX=4.3344, extentY=2.9473, extentZ=1.0626, center=<2.1567 0 -0.0001>}]
 dyn_tag_r [{shapeType=Box, extentX=4.3344, extentY=2.9473, extentZ=1.0626, center=<2.1567 0 -0.0001>}]
 dyn_tags [{shapeType=Box, extentX=6.8826, extentY=4.137, extentZ=0.6545, center=<6.5452 0 0>}]
 spine_03 [{shapeType=Capsule, length=166.5, radius=70.7574, center=<35.1482 0.552 0>, rotation=<rot -90 9.9993 0>}]
}
constraints [
 {bone1Name=dyn_scarf_f_01, bone2Name=dyn_scarf_f_MSTR, swing1LimitDegrees=45, swing2LimitDegrees=45, twistLimitDegrees=5}
 {bone1Name=dyn_scarf_02, bone2Name=dyn_scarf_f_01, swing1LimitDegrees=45, swing2LimitDegrees=45, twistLimitDegrees=5}
 {bone1Name=dyn_tags, bone2Name=dyn_scarf_f_01, swing1LimitDegrees=45, swing2LimitDegrees=45, twistLimitDegrees=5}
 {bone1Name=dyn_tag_l, bone2Name=dyn_tags, swing1LimitDegrees=45, swing2LimitDegrees=45, twistLimitDegrees=5}
 {bone1Name=dyn_tag_r, bone2Name=dyn_tags, swing1LimitDegrees=45, swing2LimitDegrees=45, twistLimitDegrees=5}
 {
  bone1Name=dyn_scarf_f_MSTR
  bone2Name=spine_03
  swing1LimitDegrees=45
  swing1Motion=Free
  swing2LimitDegrees=45
  swing2Motion=Free
  twistLimitDegrees=45
  twistMotion=Free
 }
 {bone1Name=dyn_scarf_b, bone2Name=spine_03, swing1LimitDegrees=14.1429, swing2LimitDegrees=45, twistLimitDegrees=2}
]
